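# Niagara effect for the toon plane glider, combining cloud puffs, spinning propellers and speed lines positioned from a skeletal mesh.
# User.ForwardDot scales cloud and speed line spawn rates, plus cloud offset and velocity.
# Clouds use wobbling mesh scale and two renderers, with black, enlarged, camera-offset attributes apparently forming an outline.
# Propellers pair looping sprite and mesh particles that follow a socket, while speed lines stretch with speed and have front-end variants.
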NiagaraSystem /BRCosmetics/Gadgets/Assets/VinderTech_GliderChute/Glider_Toon_Plane_Male/FX/NS_Glider_ToonPlane_Trail
dependencies {
 usedDataInterfaces [
  &NiagaraDataInterfaceArrayColor
  &NiagaraDataInterfaceCurve
  &NiagaraDataInterfaceVector2DCurve
  &NiagaraDataInterfaceArrayFloat
  &NiagaraDataInterfaceMeshRendererInfo
  &NiagaraDataInterfaceArrayFloat3
  &NiagaraDataInterfaceArrayInt32
  &NiagaraDataInterfaceColorCurve
 ]
 usedDynamicInputs []
 usedModules [&SystemState]
}
emitters {
 Clouds001 {
  emitterUpdateScript [
   &EmitterState()
   &SpawnBurst_Instantaneous(Use Spawn Probability=false, UseLoopCountLimit=false)
   &SpawnRate(SpawnRate=&Lerp_Float(A=1, B=3.5, Alpha=$User.ForwardDot), Use Spawn Probability=false)
  ]
  particleSpawnScript [
   &InitializeParticle(
    Color=<rgba 1 1 1 10>
    Lifetime=&RandomRangeFloat(Minimum=2, Maximum=2.5)
    Mesh Scale=&Multiply_VectorByFloat(
     Vector=&MakeVector(
      X=&SelectFloatFromArray(
       Array Sampling Mode=::NewEnumerator0
       Direct Array Index=?
       Float Selection Array=DataInterface() {propertyValues {FloatData [-1,1]}}
       Random Seed=0
      )
      Y=&SelectFloatFromArray(
       Array Sampling Mode=::NewEnumerator0
       Direct Array Index=?
       Float Selection Array=DataInterface() {propertyValues {FloatData [-1,1]}}
       Random Seed=0
      )
      Z=1
     )
     Float=&RandomRangeFloat(Minimum=0.3, Maximum=0.55)
    )
    Mesh Scale Mode=::NewEnumerator4
    Sprite Size Max=<60 120>
    Sprite Size Min=<30 60>
    Sprite Size Mode=::NewEnumerator2
    UsePositionOffset=false
    Write Material Random=true
   ) {module=InitializeParticle001}
   &ShapeLocation(MaskSpawnGroup=false, Override Seed=false, Sphere Radius=250, UseOwnerScale=true) {enabled=false}
   &SkeletalMeshLocation(
    Bone Sampling Mode=::NewEnumerator4
    MaskSpawnGroup=false
    Sampled Position Offset=&Lerp_Vector(
     A=&SelectVectorFromArray(
      Array Sampling Mode=::NewEnumerator0
      Direct Array Index=?
      Vector Selection Array=DataInterface() {propertyValues {FloatData [<150 -400 125>, <-150 -400 125>, <150 -400 -50>, <-150 -400 -50>]}}
      Random Seed=0
     )
     B=&SelectVectorFromArray(
      Array Sampling Mode=::NewEnumerator0
      Direct Array Index=?
      Vector Selection Array=DataInterface() {propertyValues {FloatData [<150 -650 175>, <-150 -650 175>, <150 -650 0>, <-150 -650 0>]}}
      Random Seed=0
     )
     Alpha=$User.ForwardDot
    )
    Skeletal Mesh=$User.New Skeletal Mesh
   ) {enabled=false}
   &SkeletalMeshLocation(
    Bone Sampling Mode=::NewEnumerator4
    MaskSpawnGroup=false
    Sampled Position Offset=&Add_Vector(
     A=&RandomRangeVector(Minimum=<-100 0 -50>, Maximum=<100 0 50>)
     B=&Lerp_Vector(
      A=&SelectVectorFromArray(
       Array Sampling Mode=::NewEnumerator0
       Direct Array Index=?
       Vector Selection Array=DataInterface() {propertyValues {FloatData [<175 -400 150>, <-175 -400 150>, <175 -400 -50>, <-175 -400 -50>]}}
       Random Seed=0
      )
      B=&SelectVectorFromArray(
       Array Sampling Mode=::NewEnumerator0
       Direct Array Index=?
       Vector Selection Array=DataInterface() {propertyValues {FloatData [<175 -700 200>, <-175 -700 200>, <175 -700 0>, <-175 -700 0>]}}
       Random Seed=0
      )
      Alpha=$User.ForwardDot
     )
    )
    Skeletal Mesh=$User.New Skeletal Mesh
   ) {module=SkeletalMeshLocation001}
   &AddVelocity(Velocity=<0 1 0>, Velocity Speed Scale=&Lerp_Float(A=250, B=700, Alpha=$User.ForwardDot))
   &ApplyInitialForces()
   /SetVariables(Particles.VisibilityTag=&SelectIntFromArray(
     Array Sampling Mode=::NewEnumerator1
     Direct Array Index=?
     Int Selection Array=DataInterface() {propertyValues {IntData [0,1]}}
     Random Seed=0
    )) {isSetParameters=true}
  ]
  particleUpdateScript [
   /SetVariables(Particles.RandomFloat=&RandomRangeFloat(Minimum=0, Maximum=1)) {isSetParameters=true}
   &ParticleState()
   &InheritSourceMovement(
    Applied Position Scale=&Multiply_VectorByFloat(
     Vector=<-0.05 -0.05 -0.025>
     Float=&FloatFromCurve(FloatCurve=DataInterface() {propertyValues {Curve {keys [
          {interpMode=RCIM_Cubic, tangentMode=RCTM_User, time=0, value=0}
          <at 0.2 0>
          <at 0.3 0.95>
          {interpMode=RCIM_Cubic, tangentMode=RCTM_User, time=1, value=1}
         ]}}})
    )
    ApplyForce=false
    ApplyPosition=true
    ApplyPositionFalloff=false
    Source Velocity:Vector3f=&TransformVector(Vector=$Engine.Owner.Velocity, Source Space=::World, Destination Space=::Simulation)
   )
   &ScaleColor(
    Scale Alpha=&FloatFromCurve(FloatCurve=DataInterface() {propertyValues {Curve {keys [
         {interpMode=RCIM_Cubic, tangentMode=RCTM_User, time=0.7, value=1.5}
         {arriveTangent=-2.3042, interpMode=RCIM_Cubic, leaveTangent=-2.3042, tangentMode=RCTM_User, time=1, value=0}
        ]}}})
    Scale RGB=&VectorFromFloat(Value=&FloatFromCurve(FloatCurve=DataInterface() {propertyValues {Curve {keys [
          {arriveTangent=0.3346, interpMode=RCIM_Cubic, leaveTangent=0.3346, tangentMode=RCTM_User, time=0, value=0.15}
          {arriveTangent=-0.0257, interpMode=RCIM_Cubic, leaveTangent=-0.0257, tangentMode=RCTM_User, time=1, value=0.35}
         ]}}}))
    ScaleA=true
    ScaleRGB=true
   ) {enabled=false, module=ScaleColor001}
   &SolveForcesAndVelocity(Clamp Velocity=false, Limit Acceleration=false)
   &ScaleMeshSize(Scale Factor=&VectorFromFloat(Value=&FloatFromCurve(FloatCurve=DataInterface() {propertyValues {Curve {keys [
          {arriveTangent=7.5557, interpMode=RCIM_Cubic, leaveTangent=7.5557, tangentMode=RCTM_User, time=0, value=0}
          {arriveTangent=0.0349, interpMode=RCIM_Cubic, leaveTangent=0.0349, tangentMode=RCTM_User, time=0.1, value=1.15}
          {arriveTangent=-1.3734, interpMode=RCIM_Cubic, leaveTangent=1.4248, tangentMode=RCTM_Break, time=0.2, value=0.8}
          {arriveTangent=0.0401, interpMode=RCIM_Cubic, leaveTangent=0.0401, tangentMode=RCTM_User, time=0.275, value=1.1}
          {interpMode=RCIM_Cubic, leaveTangent=2.5288, tangentMode=RCTM_Break, time=0.35, value=0.9}
          {arriveTangent=0.026, interpMode=RCIM_Cubic, leaveTangent=0.026, tangentMode=RCTM_User, time=0.4, value=1}
          {arriveTangent=-1.3077, interpMode=RCIM_Cubic, leaveTangent=-1.3077, time=0.8, value=0.875}
          <at 1 0 RCIM_Cubic>
         ]}}}))) {module=ScaleMeshSize001}
   &DynamicMaterialParameters(
    Index 0 Param 1=1
    Index 0 Param 2=1
    Index 0 Param 4=1
    Module.Param0WriteEnabled=true
    Module.Param1WriteEnabled=true
    Module.Param2WriteEnabled=true
    Module.Param3WriteEnabled=true
   ) {module=DynamicMaterialParameters001}
   &FN_Mesh_Attributes(
    Color=<rgba 0 0 0 1>
    Create Color Attribute=true
    Create Scale Attribute=true
    Scale=&Multiply_VectorByFloat(Vector=$Particles.Scale, Float=1.1)
    Write to Module Namespace=true
   )
   /SetVariables(Particles.R2_CameraOffset:NiagaraFloat=&Add_Float(
     A=-5
     B=&FloatFromCurve(
      FloatCurve=DataInterface() {propertyValues {Curve {keys [
          {arriveTangent=7.5557, interpMode=RCIM_Cubic, leaveTangent=7.5557, tangentMode=RCTM_User, time=0, value=0}
          {arriveTangent=0.0349, interpMode=RCIM_Cubic, leaveTangent=0.0349, tangentMode=RCTM_User, time=0.1, value=1.15}
          {arriveTangent=-1.3734, interpMode=RCIM_Cubic, leaveTangent=1.4248, tangentMode=RCTM_Break, time=0.2, value=0.8}
          {arriveTangent=0.0401, interpMode=RCIM_Cubic, leaveTangent=0.0401, tangentMode=RCTM_User, time=0.275, value=1.1}
          {interpMode=RCIM_Cubic, leaveTangent=2.5288, tangentMode=RCTM_Break, time=0.35, value=0.9}
          {arriveTangent=0.026, interpMode=RCIM_Cubic, leaveTangent=0.026, tangentMode=RCTM_User, time=0.4, value=1}
          {arriveTangent=-1.3077, interpMode=RCIM_Cubic, leaveTangent=-1.3077, time=0.8, value=0.875}
          <at 1 0 RCIM_Cubic>
         ]}}}
      Scale Curve=-20
     )
    )) {isSetParameters=true}
  ]
  rendererClasses [&NiagaraMeshRendererProperties]
  renderers [
   {rendererClass=&NiagaraMeshRendererProperties, rendererIndex=0}
   {rendererClass=&NiagaraMeshRendererProperties, rendererIndex=1}
  ]
 }
 Propeller {
  emitterUpdateScript [
   &EmitterState(Life Cycle Mode=::NewEnumerator1, Loop Behavior=::NewEnumerator1, Loop Duration Mode=::NewEnumerator1, UseLoopDelay=false)
   &SpawnBurst_Instantaneous(Spawn Count=2, Use Spawn Probability=false, UseLoopCountLimit=false)
  ]
  particleSpawnScript [
   &InitializeParticle(
    Color=<rgba 0.3559 0.3559 0.3559 10>
    Lifetime=0.5
    Sprite Size Mode=::NewEnumerator3
    Uniform Sprite Size=78
    UsePositionOffset=false
    Write Material Random=true
   )
   &FN_SetSpawnGroup(Number of Spawn Groups=2)
  ]
  particleUpdateScript [
   &ParticleState(
    Kill Particles When Lifetime Has Elapsed=false
    Let Infinitely Lived Particles Die When Emitter Deactivates=true
    Loop Particles Lifetime=true
   )
   &SkeletalMeshLocation(Bone / Socket Index=2, Bone Sampling Mode=::NewEnumerator4, MaskSpawnGroup=false, Skeletal Mesh=$User.New Skeletal Mesh)
   &FN_Radial_Material(Base Sprite Offset X=0.5, Base Sprite Tile U=10, Base Sprite Tile V=0.02, Depth Fade=0, Radial Mask Power=1.5, Radial Mask Strength=1.5)
   &SpriteRotationRate(Rotation Rate=50000)
   /SetVariables(Particles.SpriteFacing=&RotateVectorByQuaternion(
     Vector To Rotate=&MakeVector(X=0, Y=&MaskFloatBySpawnGroup(Masked Float=-1, Passthrough Float=1), Z=0)
     Quaternion=$Particles.SkeletalMeshLocation.SampledRotation
    )) {isSetParameters=true}
  ]
  rendererClasses [&NiagaraSpriteRendererProperties]
  renderers [{rendererClass=&NiagaraSpriteRendererProperties}]
 }
 Propeller2 {
  emitterUpdateScript [
   &EmitterState(Life Cycle Mode=::NewEnumerator1, Loop Behavior=::NewEnumerator1, Loop Duration Mode=::NewEnumerator1, UseLoopDelay=false)
   &SpawnBurst_Instantaneous(Spawn Count=2, Use Spawn Probability=false, UseLoopCountLimit=false)
  ]
  particleSpawnScript [
   &InitializeParticle(
    Color=<rgba 0 0 0 1.2>
    Lifetime=0.5
    Mesh Scale Mode=::NewEnumerator3
    Mesh Uniform Scale=4.5
    UsePositionOffset=false
    Write Material Random=true
   )
   &FN_SetSpawnGroup(Number of Spawn Groups=2)
   &InitialMeshOrientation(
    Mesh Orientation Mode=::NewEnumerator1
    Rotation=&MaskVectorBySpawnGroup(Masked Vector=<0.25 0 0>, Passthrough Vector=<-0.25 0 0>)
    Use Rotation Vector=true
   )
  ]
  particleUpdateScript [
   &ParticleState(
    Kill Particles When Lifetime Has Elapsed=false
    Let Infinitely Lived Particles Die When Emitter Deactivates=true
    Loop Particles Lifetime=true
   )
   &SkeletalMeshLocation(Bone / Socket Index=2, Bone Sampling Mode=::NewEnumerator4, MaskSpawnGroup=false, Skeletal Mesh=$User.New Skeletal Mesh)
   &FN_Radial_Material(Base Sprite Tile U=4, Base Sprite Tile V=0.1, Depth Fade=0, Radial Mask Power=1, Radial Mask Tile V=3, Sprite Spin Speed=1)
   &UpdateMeshOrientation(Rotation Vector=<0 0 1>)
   &CameraOffset(Camera Offset Amount=10)
  ]
  rendererClasses [&NiagaraMeshRendererProperties]
  renderers [{rendererClass=&NiagaraMeshRendererProperties, rendererIndex=0}]
 }
 Speedlines001 {
  emitterUpdateScript [
   &EmitterState()
   &SpawnRate(SpawnRate=&Lerp_Float(A=0, B=50, Alpha=$User.ForwardDot), Use Spawn Probability=false)
  ]
  particleSpawnScript [
   /Niagara/Modules/Spawn/Initialization/InitializeParticle(
    Color=&SelectLinearColorFromArray(
     Array Sampling Mode=::NewEnumerator0
     Direct Array Index=?
     Color Selection Array=DataInterface() {propertyValues {ColorData [<rgba 0 0 0 4>, <rgba 1 1 1 4>]}}
     Random Seed=0
    )
    Lifetime=&FN_FrontEndFloatModifier(
     Gameplay=&UniformRangedFloat(Minimum=0.2, Maximum=0.35)
     Front End=&UniformRangedFloat(Minimum=0.1, Maximum=0.2)
    )
    Module.Write Color=true
    Module.Write Lifetime=true
    Module.Write Mass=true
    Module.Write Material Random=true
    Module.Write Position=false
    Module.Write Scale=false
    Module.Write SpriteRotation=false
    Module.Write SpriteSize=true
    Position=&SimulationPosition()
    Sprite Size=&UniformRangedVector2D(Minimum=<1 4>, Maximum=<3 6>)
   )
   &CylinderLocation(
    Cylinder Height=200
    Cylinder Radius=&UniformRangedFloat(Minimum=150, Maximum=200)
    MaskSpawnGroup=false
    Module.Enable Lathe Profile=false
    Module.Override Randomness=false
    Module.Surface Only=true
    Non Uniform Scale:Vector3f=<1 1 0.3>
    Offset=<0 0 35>
    Orientation Axis:ENiagaraOrientationAxis=::YAxis
    Use Endcaps In Surface Only Mode=false
   )
   &SkeletalMeshLocation(
    Bone Sampling Mode=::NewEnumerator4
    MaskSpawnGroup=false
    Sampled Position Offset:NiagaraPosition=$Particles.Position
    Skeletal Mesh=$User.New Skeletal Mesh
   )
   &AddVelocity(Velocity=&FN_FrontEndVector3Modifier(Gameplay=<0 1000 0>, Front End=<0 3000 0>))
   &AddVelocityFromPoint(
    Module.Use Falloff Distance=false
    Velocity Strength=&Multiply_Float(
     A=&FN_FrontEndFloatModifier(Gameplay=1, Front End=2)
     B=&UniformRangedFloat(Minimum=500, Maximum=700)
    )
   )
  ]
  particleUpdateScript [
   &ParticleState()
   &SolveForcesAndVelocity(Clamp Velocity=false, Limit Acceleration=false)
   &ScaleSpriteSize(Scale Factor=&Vector2DFromCurve(
     Vector2Curve=DataInterface() {propertyValues {XCurve {keys [<at 0 0.5 RCIM_Cubic>,<at 0.5 1 RCIM_Cubic>]}, YCurve {keys [<at 0 0 RCIM_Cubic>,<at 0.5 1 RCIM_Cubic>]}}}
     ScaleCurve=&Vector2DFromFloat(Value=1)
    ))
   &ScaleSpriteSizeBySpeed(Max Scale Factor=<1 40>, Min Scale Factor=<0.01 0.05>, Velocity Threshold=1)
   &ScaleColor(
    Scale Alpha=&FloatFromCurve(
     FloatCurve=DataInterface() {propertyValues {Curve {keys [
         {arriveTangent=0.1307, interpMode=RCIM_Cubic, leaveTangent=0.1307, tangentMode=RCTM_User, time=0, value=0}
         <at 0.3 1>
         {arriveTangent=-2.221, interpMode=RCIM_Cubic, leaveTangent=-2.221, tangentMode=RCTM_User, time=1, value=0}
        ]}}}
     Scale Curve=&FN_FrontEndFloatModifier(Gameplay=1, Front End=1)
    )
    ScaleA=true
    ScaleRGB=true
   )
  ]
  rendererClasses [&NiagaraSpriteRendererProperties]
  renderers [{rendererClass=&NiagaraSpriteRendererProperties}]
 }
}
summary {
 systemName=NS_Glider_ToonPlane_Trail
 emitters [
  {emitterName=Speedlines001, rendererClasses [&NiagaraSpriteRendererProperties]}
  {emitterName=Propeller, rendererClasses [&NiagaraSpriteRendererProperties]}
  {emitterName=Propeller2, rendererClasses [&NiagaraMeshRendererProperties]}
  {emitterName=Clouds001, rendererClasses [&NiagaraMeshRendererProperties]}
 ]
 userVariables [
  User.bIsFrontEnd {type=NiagaraBool}
  User.bIsFrontEndPreview {type=NiagaraBool}
  User.ForwardDot {type=NiagaraFloat, defaultValue=0, description=""}
  User.New Skeletal Mesh {
   type=NiagaraDataInterfaceSkeletalMesh
   defaultValue {
    struct=NiagaraExt_VariableValue_DataInterface
    value {dataInterface=&self:NiagaraDataInterfaceSkeletalMesh_1}
   }
  }
 ]
}
