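# Physics asset for the Toon Plane Male glider cosmetic in the VinderTech GliderChute gadget set.
# It defines a single box body on the upper chassis bone, offset and rotated to cover the plane mesh.
# There are no constraints, so the glider has one collision volume with no jointed parts.
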PhysicsAsset /BRCosmetics/Gadgets/Assets/VinderTech_GliderChute/Glider_Toon_Plane_Male/Meshes/Toon_Plane_Male_Glider_Physics
bodies {JNT_chassisUpper [{shapeType=Box, extentX=408.3574, extentY=126.3366, extentZ=313.8411, center=<0.0218 19.497 6.4948>, rotation=<rot 0 1.6211 -86.8979>}]}
constraints []
